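# Animation Blueprint for the EmeraldGlassTransform large male player head, correcting neck placement as the head tilts.
# Each update derives back and forward alphas from head-versus-spine_05 twist, plus a crouching head-back alpha, all zeroed in the frontend unless emoting.
# These drive three chained additive neck_01 offsets on the input pose, with the crouch one adding roll and interpolated blending.
# The head's Control Rig runs last, under the same frontend/emote gate.
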AnimBlueprint /Game/Characters/Player/Male/Large/Heads/M_LRG_EmeraldGlassTransform_Head/Meshes/M_LRG_EmeraldGlassTransform_Head_Export_Skeleton_AnimBlueprint
target_skeleton=/Game/Characters/Player/Common/Fortnite_Base_Head/Fortnite_Base_Head_Export_Skeleton
graphs [
 AnimGraph {nodes [
   n1 = "Transform (Modify) Bone - Bone: neck_01"(
    ComponentPose=Local To Component(LocalPose=Input Pose())
    Alpha=Get headRotBack() -> (headRotBack:double)
    Translation="0,-4,2"
    Rotation="0,0,0"
    Scale="1,1,1"
   ) {translationMode=BMM_Additive, alphaBoolBlend {blendOption=Linear}, boneToModify {boneName=neck_01}, translation {y=-4, z=2}}
   n2 = "Transform (Modify) Bone - Bone: neck_01"(ComponentPose=@n1, Alpha=Get HeadRotFwd() -> (HeadRotFwd:float), Translation="0,2,-2", Rotation="0,0,0", Scale="1,1,1") {translationMode=BMM_Additive, alphaBoolBlend {blendOption=Linear}, boneToModify {boneName=neck_01}, translation {y=2, z=-2}}
   n3 = "Transform (Modify) Bone - Bone: neck_01"(ComponentPose=@n2, Alpha=Get isCrouchingHeadBack() -> (isCrouchingHeadBack:double), Translation="0,0,2", Rotation="0,0,15", Scale="1,1,1") {
    rotationMode=BMM_Additive
    translationMode=BMM_Additive
    alphaBoolBlend {blendOption=Linear}
    alphaScaleBiasClamp {bInterpResult=true, interpSpeedDecreasing=20, interpSpeedIncreasing=20}
    boneToModify {boneName=neck_01}
    rotation {roll=15}
    translation {z=2}
   }
   n4 = Control Rig(Source=Component To Local(ComponentPose=@n3), Alpha=Get NoFrontendIsEmoting() -> (NoFrontendIsEmoting:float)) {
    alphaBoolBlend {blendOption=Linear}
    controlRigAssetReference {blueprintRigClass=&/M_LRG_EmeraldGlassTransform_Head_CtrlRig.M_LRG_EmeraldGlassTransform_Head_CtrlRig_C}
   }
   Output Pose(Result=@n4) {layerGroup=None}
  ]}
 EventGraph {nodes [
   Event BlueprintUpdateAnimation()
   n1 = Get bIsInFrontEnd()
   n2 = "To Float (Boolean)"(InBool=OR Boolean(A=NOT Boolean(A=@n1), B=Get bIsPlayingEmote()))
   n3 = GetOwnerSkeletalMeshComponent()
   > Set headRotBack(headRotBack:double=float * float(
     A=GetAlphaFromReferenceTransformDelta(SkeletalMeshComponent=@n3, BoneName=head, BaseName=spine_05, InputRange=(X=-20,Y=-50), TwistAxis="1,0,0")
     B=@n2
    ) {is=title K2Node_PromotableOperator}) -> (Output_Get:double)
   n4 = GetOwnerSkeletalMeshComponent()
   > Set HeadRotFwd(HeadRotFwd:float=float * float(
     A=GetAlphaFromReferenceTransformDelta(SkeletalMeshComponent=@n4, BoneName=head, BaseName=spine_05, InputRange=(X=-20,Y=20), TwistAxis="1,0,0")
     B=@n2
    )) -> (Output_Get:float)
   > Set NoFrontendIsEmoting(NoFrontendIsEmoting:float=@n2) -> (Output_Get:float)
   > n5 = Set Is Crouching(Is Crouching:double="To Float (Boolean)"(InBool=Get bIsCrouching())) -> (Output_Get:double)
   n6 = OR Boolean(A=NOT Boolean(A=Get bIsInFrontEnd()), B=Get bIsPlayingEmote())
   > Set isCrouchingHeadBack(isCrouchingHeadBack:double=float * float!(A=@n5, B=Get headRotBack(), C:double="To Float (Boolean)"(InBool=@n6)) -> (ReturnValue)) -> (Output_Get:double)
   TryGetPawnOwner()
  ]}
]
variables [
 headRotBack {type=number}
 HeadRotFwd {type=number}
 Is Crouching {type=number}
 isCrouchingHeadBack {type=number}
 NoFrontendIsEmoting {type=number}
]
